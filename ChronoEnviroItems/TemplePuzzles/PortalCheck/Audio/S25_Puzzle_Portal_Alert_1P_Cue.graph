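# Sound cue for the first-person (1P) portal alert in a Season 25 temple puzzle, judging by its name.
# A single wave player feeds a Modulator at default settings, boosted slightly by the cue volume and routed to the gadgets and items sound class.
# A concurrency override caps simultaneous copies and sets a retrigger time.
# Low-pass and occlusion attenuation values are stored, but the attenuation override flag is not set.
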SoundCue /ChronoEnviroItems/TemplePuzzles/PortalCheck/Audio/S25_Puzzle_Portal_Alert_1P_Cue
root=Modulator_0
nodes [
 {id=Modulator_0, children [WavePlayer_1]}
 {id=WavePlayer_1, properties {soundWaveAssetPtr=&/S25_Puzzle_Portal_Alert_1P}}
]
properties {
 bOverrideConcurrency=true
 maxDistance=2097152
 soundClassObject=&Fort_Gadgets_Items
 volumeMultiplier=1.3
 attenuationOverrides {
  bAttenuateWithLPF=true
  bEnableOcclusion=true
  lPFFrequencyAtMax=1000
  lPFFrequencyAtMin=15000
  lPFRadiusMax=3600
  lPFRadiusMin=500
  occlusionLowPassFilterFrequency=2000
  stereoSpread=350
  attenuationShapeExtents {x=640.6504}
 }
 concurrencyOverrides {maxCount=2, retriggerTime=0.3}
}
