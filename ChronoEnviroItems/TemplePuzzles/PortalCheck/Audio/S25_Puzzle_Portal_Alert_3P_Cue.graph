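# Sound cue for a temple puzzle portal alert, the third-person variant by its 3P name, in the gadgets and items sound class.
# A single wave player feeds a Modulator left at default settings for pitch and volume variation.
# Inline attenuation overrides the assigned door attenuation asset, adding distance low-pass filtering and occlusion.
# A concurrency override caps simultaneous instances and sets a short retrigger time.
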SoundCue /ChronoEnviroItems/TemplePuzzles/PortalCheck/Audio/S25_Puzzle_Portal_Alert_3P_Cue
root=Modulator_0
nodes [
 {id=Modulator_0, children [WavePlayer_0]}
 {id=WavePlayer_0, properties {soundWaveAssetPtr=&/S25_Puzzle_Portal_Alert_3P}}
]
properties {
 attenuationSettings=/Game/Sounds/Attenuation/Fort_Doors_Att
 bOverrideAttenuation=true
 bOverrideConcurrency=true
 maxDistance=2800
 soundClassObject=&Fort_Gadgets_Items
 volumeMultiplier=1.3
 attenuationOverrides {
  bAttenuateWithLPF=true
  bEnableOcclusion=true
  falloffDistance=1500
  lPFFrequencyAtMax=1000
  lPFFrequencyAtMin=15000
  lPFRadiusMax=3600
  lPFRadiusMin=1550
  occlusionLowPassFilterFrequency=2000
  occlusionTraceChannel=ECC_GameTraceChannel6
  stereoSpread=320
  attenuationShapeExtents {x=1300}
 }
 concurrencyOverrides {maxCount=2, retriggerTime=0.3}
}
