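# Sound cue for the success sting of the temple puzzle portal check in ChronoEnviroItems.
# A single wave player plays the S25_Sting_Puzzle_Portal sound through a modulator that fixes pitch and lowers volume to a constant level, so there is no random variation.
# The cue routes to an indoor ambient sound class with a slightly reduced overall volume and a very large maximum distance.
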SoundCue /ChronoEnviroItems/TemplePuzzles/PortalCheck/Audio/Sting_Puzzle_Portal_Success_Cue
root=Modulator_0
nodes [
 {id=Modulator_0, children [WavePlayer_1], properties {pitchMax=1, pitchMin=1, volumeMax=0.6, volumeMin=0.6}}
 {id=WavePlayer_1, properties {soundWaveAssetPtr=&/S25_Sting_Puzzle_Portal}}
]
properties {maxDistance=2097152, soundClassObject=&Fort_Ambient_Indoors_Athena, volumeMultiplier=0.902}
