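# Translucent unlit material for the eye glow effect in the Chrono temple PortalCheck puzzle, enabled for Niagara mesh particles.
# Particle color RGB passes through the shared MF_FX_EmissiveAdjustments function for the glow color.
# Particle alpha is multiplied by a mask from inverted vertex red, raised to the VertexColor dynamic parameter, so the particle system can control the falloff.
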Material /ChronoEnviroItems/TemplePuzzles/PortalCheck/FX/M_Chrono_Temple_PortalCheck_EyeGlow
graph {
 properties {
  bUsedWithNiagaraMeshParticles=true
  blendMode=BLEND_Translucent
  pixelDepthOffsetMode=PDOM_Legacy
  shadingModel=MSM_Unlit
  usedShadingModels=Unlit
 }
 nodes [
  n1 = ParticleColor()
  MaterialFunctionCall(Emissive=@n1.RGB) {materialFunction=&MF_FX_EmissiveAdjustments}
  n2 = VertexColor()
  Multiply(A=@n1.A, B=Power(Base=OneMinus(None=@n2.R), Exp=DynamicParameter() {paramNames [VertexColor,Param2,Param3,Param4]}.VertexColor))
 ]
}
